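; Gameplay ability for a quest buff item in Mash playlists, with an item-removal path that its name ties to non-Mash play.
; Server-side, it waits for playlist data, then on Mash swaps boon items, flagging mid-match expertise if the old boon was held.
; That flag selects the shard grant amount and gates starting ammo.
; It also replaces the tagged buff GameplayEffect and ends when a configured boon tag is added.
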(event Ability|EventActivateAbility
  (bind _buffge_tag (Variables|Default|GetBuffGE_Tag))
  (bind _ge_toapply (Variables|Default|GetGE_ToApply))
  (bind _boontagtoendability (Variables|Default|GetBoonTagToEndAbility))
  (Networking|CanExecuteCosmeticEvents
    (:False
      (bind (_gamestate _playlist _playlistcontexttags) (Class|FortAsyncActionCurrentPlaylistReady|PlaylistDataReadyAsync)
        (:OnReady
          (if (GameplayTags|HasTag _playlistcontexttags "(TagName=\"Athena.Playlist.Mash\")" true)
            (bind _outfortplayerpawn (CallFunction|GetAvatarasValidFortPlayerPawn)
              (:ValidFortPlayerPawn
                (Utilities|FlowControl|Sequence
                  (:then_0
                    (Utilities|IsValid (Fortnite|GetFortPlayerControllerfromActor _outfortplayerpawn)
                      (:Is%20Valid
                        (if (ScalableFloat|GetValueasBool (Variables|Default|GetHF_BoonGrantEnabled) (Variables|Default|GetLevelForEffect))
                          (bind _boontogrant (Variables|Default|GetBoonToGrant)
                            (:then
                              (Inventory|GiveItemtoInventoryOwner (Utilities|Casting|CastToFortInventoryOwnerInterface (Fortnite|GetFortPlayerControllerfromActor _outfortplayerpawn)) _boontogrant)
                              (bind _boontoremove (Variables|Default|GetBoonToRemove)
                                (:then
                                  (bind _returnvalue (InventoryManagement|GetItemCountforItemDefinition (Utilities|Casting|CastToFortInventoryOwnerInterface (Fortnite|GetFortPlayerControllerfromActor _outfortplayerpawn)) _boontoremove))
                                  (if (> _returnvalue 0)
                                    (Inventory|RemoveItemfromInventoryOwner (Utilities|Casting|CastToFortInventoryOwnerInterface (Fortnite|GetFortPlayerControllerfromActor _outfortplayerpawn)) _boontoremove)
                                    (Variables|Default|SetbGainedExpertiseMidMatch true)))
                                (:else)))
                            (:else))))))
                  (:then_1
                    (CallFunction|GrantStartingItems _outfortplayerpawn))
                  (:then_2
                    (if (GameplayTags|HasAnyMatchingGameplayTags _outfortplayerpawn _buffge_tag)
                      (Ability|RemoveGameplayEffectFromOwnerWithGrantedTags _buffge_tag)
                      (else
                        _))
                    (if (Utilities|IsValidClass _ge_toapply)
                      (bind _returnvalue_2 (Ability|ApplyGameplayEffectToOwner _ge_toapply (Math|Float|Truncate (Variables|Default|GetLevelForEffect))))
                      (Variables|Default|SetGE_ActiveHandle _returnvalue_2)))
                  (:then_3
                    (if (GameplayTags|IsGameplayTagValid _boontagtoendability)
                      (Ability|Tasks|WaitGameplayTagAdd _boontagtoendability
                        (:Added
                          (Ability|EndAbility)))))
                  (:then_4
                    _))))
            (else
              (Utilities|FlowControl|Sequence
                (:then_0
                  _)
                (:then_1
                  _))))
          (bind _outfortplayerpawn_2 (CallFunction|GetAvatarasValidFortPlayerPawn)
            (:ValidFortPlayerPawn
              (Utilities|IsValid (Fortnite|GetFortPlayerControllerfromActor _outfortplayerpawn_2)
                (:Is%20Valid
                  (Inventory|RemoveItemfromInventoryOwner (Fortnite|GetFortPlayerControllerfromActor _outfortplayerpawn_2) (Variables|Default|GetItemToRemove)))))))))))

(event Ability|EventOnEndAbility
  (params (bWasCancelled :type Boolean))
  (|CallParentFunction bWasCancelled))

(fn GrantStartingItems
  (params (PlayerPawn :type Fort%20Player%20Pawn%20Object%20Reference))
  (bind _playerpawn (Variables|GetPlayerPawn))
  (bind _returnvalue (Fortnite|GetFortPlayerControllerfromActor _playerpawn))
  (bind _playerpawn_2 (Variables|GetPlayerPawn))
  (bind _returnvalue_2 (Fortnite|GetFortPlayerControllerfromActor _playerpawn_2))
  (Utilities|FlowControl|Sequence
    (:then_0
      (if (ScalableFloat|GetValueasBool (Variables|Default|GetHF_ShardsGrantEnabled) (Variables|Default|GetLevelForEffect))
        (bind _outloottodrop (Loot|PickLootDrops (Variables|Default|GetTierGroup_Shards) (Loot|GetLootLevel) -1 "(GameplayTags=)"))
        (Utilities|IsValid _returnvalue_2
          (:Is%20Valid
            (Inventory|GiveItemtoInventoryOwner :InventoryOwner (Utilities|Casting|CastToFortInventoryOwnerInterface _returnvalue_2) :ItemDefinition (Utilities|Casting|CastToFortWorldItemDefinition (Class|FortItemFunctionLibrary|BreakFortItemEntry (Utilities|Array|Get(aref) _outloottodrop))) :NumberToGive (Math|Integer|SelectInt (ScalableFloat|GetValueasInteger (Variables|Default|GetHF_ShardsToGrant) (Variables|Default|GetLevelForEffect)) (ScalableFloat|GetValueasInteger (Variables|Default|GetHF_ShardsAccumulativeGrant) (Variables|Default|GetLevelForEffect)) (Variables|Default|GetbGainedExpertiseMidMatch)) :bNotifyPlayer true :ItemLevel -1 :PickupInstigatorHandle -1 :bUseItemPickupAnalyticEvent false :WeaponAmmoOverride -1)))))
    (:then_1
      (if (and (ScalableFloat|GetValueasBool (Variables|Default|GetHF_BoonGrantAmmo) (Variables|Default|GetLevelForEffect)) (not (Variables|Default|GetbGainedExpertiseMidMatch)))
        (bind (_keypin _valuepin) (Utilities|Map|ForEachLoop(Map) (Variables|Default|GetAmmoAmountToGive)))
        (Utilities|IsValid _returnvalue
          (:Is%20Valid
            (Inventory|GiveItemtoInventoryOwner :InventoryOwner (Utilities|Casting|CastToFortInventoryOwnerInterface _returnvalue) :ItemDefinition _keypin :NumberToGive (ScalableFloat|GetValueasInteger _valuepin (Variables|Default|GetLevelForEffect)) :bNotifyPlayer true :ItemLevel -1 :PickupInstigatorHandle -1 :bUseItemPickupAnalyticEvent false :WeaponAmmoOverride -1)))))))
